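# NS_BoostInk is the Niagara effect for the Ink vehicle boost cosmetic, an ink-like exhaust trail that is near-black unless painted.
# Ribbons, ink puffs and SubUV soft smoke spawn per boost socket on the skeletal mesh, move along negative X and inherit decaying source movement.
# Painted-variant overrides supply tints, while color variant ID sets ink puff visibility tag and material parameters.
# A frontend preview flag retunes motion, lifetime and puff shaping.
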NiagaraSystem /VehicleCosmetics/Boosts/Boost_Ink/FX/NS_BoostInk
dependencies {
 usedDataInterfaces [
  &NiagaraDataInterfaceCurve
  &NiagaraDataInterfaceVectorCurve
  &NiagaraDataInterfaceSpriteRendererInfo
  &NiagaraDataInterfaceArrayInt32
  &NiagaraDataInterfaceArrayFloat
  &NiagaraDataInterfaceSkeletalMesh
  &NiagaraDataInterfaceArrayFloat3
 ]
 usedDynamicInputs [&SelectVectorFromArray]
 usedModules [
  &self:SystemSpawnScript.SystemScriptSource.SystemScriptGraph.NiagaraNodeAssignment_0.SetVariables_#1
  &self:SystemSpawnScript.SystemScriptSource.SystemScriptGraph.NiagaraNodeAssignment_1.SetVariables_#2
  &SystemState
  &self:SystemSpawnScript.SystemScriptSource.SystemScriptGraph.NiagaraNodeAssignment_2.SetVariables_#3
 ]
}
emitters {
 NE_VC_Boost_Ribbon_Parent {
  emitterUpdateScript [
   &EmitterState()
   &VC_FilteredSocketCount()
   &VC_RateQualityHandler()
   &VC_SpawnRibbon_v1(Number of Ribbons=$Emitter.PortCount, SpawnRate=60)
  ]
  particleSpawnScript [
   &VC_PaintedVariantsOverride(
    BurntSienna=<rgba 0.1406 0.0218 0.0066 0.5826>
    Cobalt=<rgba 0 0.0185 1 0.6783>
    ForestGreen=<rgba 0 0.2969 0.015 0.4522>
    Grey=<rgba 0.1406 0.1406 0.1406 0.8435>
    Lime=<rgba 0.2943 0.5885 0 0.6522>
    Orange=<rgba 0.6406 0.2235 0 0.6087>
    Pink=<rgba 0.5052 0.0368 0.3349 0.6696>
    Purple=<rgba 0.2057 0 0.6795 0.513>
    Saffron=<rgba 0.6042 0.4632 0 0.6174>
    SkyBlue=<rgba 0 0.357 0.5365 0.5739>
    TitaniumWhite=<rgba 0.5833 0.5833 0.5833 0.6522>
   )
   &VC_InitializeRibbon_v1(
    Color=&MakeCustomLinearColorFromBool(Bool=$User.bIsPainted, True Color=$Particles.Color, False Color=<rgba 0.0156 0.0156 0.0156 1>)
    Enable Skeletal Position=true
    Lifetime=1.2
    Offset Amount=50
    Offset Axis:ENiagaraOrientationAxis=::XAxis
    Ribbon Width=180
   )
   &AddVelocity(Velocity=/Niagara/DynamicInputs/TypeConversions/MakeCustomPositionFromBool(Bool:NiagaraBool=$User.bIsFrontendPreview, True Position:NiagaraPosition=<-1600 0 0>, False Position:NiagaraPosition=<-1700 0 0>)) {module=AddVelocity001}
   /VehicleCosmetics/FX/VC_FXParents/Modules/DLR_ProductPreviews(GameModeID:NiagaraInt32=$System.GameModeID, bPreviewVehicleSpeed:NiagaraBool=$System.bPreviewVehicleSpeed)
  ]
  particleUpdateScript [
   &ParticleState()
   &InheritSourceMovement(
    Applied Position Scale=&VectorFromFloat(Value=&FloatFromCurve(FloatCurve=DataInterface() {propertyValues {Curve {
         defaultValue=1
         keys [
          {arriveTangent=-1.7885, interpMode=RCIM_Cubic, leaveTangent=-1.7885, tangentMode=RCTM_User, time=0.05, value=1}
          {interpMode=RCIM_Cubic, tangentMode=RCTM_User, time=1, value=0}
         ]
        }}}))
    ApplyForce=false
    ApplyPosition=true
    ApplyPositionFalloff=false
   )
   &SolveForcesAndVelocity(Clamp Velocity=false, Limit Acceleration=false)
   &DynamicMaterialParameters(
    Index 2 Param 1=-1
    Index 2 Param 2=1
    Index 2 Param 3=-1
    Index 2 Param 4=1
    Module.Index0Float0Write=true
    Module.Index0Float1Write=true
    Module.Index0Float2Write=true
    Module.Index0Float3Write=true
    Module.Index2Float0Write=true
    Module.Index2Float1Write=true
    Module.Index2Float2Write=true
    Module.Index2Float3Write=true
    Write Parameter Index 2=::NewEnumerator1
   )
   &VC_UpdateRibbon_v1(Enable Skeletal Position=true)
  ]
  rendererClasses [&NiagaraRibbonRendererProperties]
  renderers [
   {rendererClass=&NiagaraRibbonRendererProperties}
   {rendererClass=&NiagaraRibbonRendererProperties, rendererIndex=1}
  ]
 }
 NE_VC_InkPuffs {
  emitterUpdateScript [
   &EmitterState()
   &VC_FilteredSocketCount()
   &SpawnRate(SpawnRate=180, Use Spawn Probability=false)
  ]
  particleSpawnScript [
   &VC_PaintedVariantsOverride(
    BurntSienna=<rgba 0.0938 0.0229 0 1>
    Cobalt=<rgba 0.0195 0.0261 0.375 1>
    Crimson=<rgba 0.125 0 0.0034 1>
    ForestGreen=<rgba 0 0.1823 0.0092 1>
    Grey=<rgba 0.1042 0.1042 0.1042 1>
    Orange=<rgba 0.4948 0.1096 0 1>
    Pink=<rgba 0.2031 0.0243 0.1381 1>
    Purple=<rgba 0.0696 0 0.3646 1>
    Saffron=<rgba 0.4531 0.3474 0 1>
    SkyBlue=<rgba 0 0.2357 0.3542 1>
    TitaniumWhite=<rgba 0.4323 0.4323 0.4323 1>
   )
   /SetVariables(Particles.VisibilityTag=&SelectIntFromArray(Direct Array Index=$System.ColorVariantID, Int Selection Array=DataInterface() {propertyValues {IntData [0,0,0,1,1,1,1,1,1,1,1,1,1,1,1]}})) {isSetParameters=true}
   &InitializeParticle(
    Color=&MakeCustomLinearColorFromBool(Bool=$User.bIsPainted, True Color=$Particles.Color, False Color=<rgba 0.0156 0.0156 0.0156 1>)
    Lifetime=&MakeCustomFloatFromBool(
     Bool=$User.bIsFrontendPreview
     True Float=&RandomRangeFloat(Minimum=0.65, Maximum=0.8)
     False Float=&RandomRangeFloat(Minimum=1, Maximum=1.4)
    )
    Position Mode=::NewEnumerator0
    Sprite Rotation Mode=::NewEnumerator2
    Sprite Size Mode=::NewEnumerator3
    Sprite UV Mode=::NewEnumerator3
    Uniform Sprite Size=&MakeCustomFloatFromBool(
     Bool=$User.bIsFrontendPreview
     True Float=&RandomRangeFloat(Minimum=55, Maximum=75)
     False Float=&RandomRangeFloat(Minimum=58, Maximum=75)
    )
    Write Material Random=true
   )
   &PartitionParticles(Number Of Partitions=$Emitter.PortCount)
   &SkeletalMeshLocation(
    Bone / Socket Index=$Particles.PartitionParticles.Partition
    Bone / Socket Sampling Mode=::NewEnumerator5
    MaskSpawnGroup=false
    Mesh Sampling Type=::NewEnumerator6
    Skeletal Mesh=$System.SkeletalMesh
   )
   &AddVelocity(Velocity=&MakeCustomVectorFromBool(
     Bool=$User.bIsFrontendPreview
     True Vector=&RandomRangeVector(Minimum=<-432.1997 0 0>, Maximum=<-800 0 0>)
     False Vector=&RandomRangeVector(Minimum=<-1600 0 0>, Maximum=<-1800 0 0>)
    ))
   &ShapeLocation(
    Cylinder Height=&MakeCustomFloatFromBool(Bool=$User.bIsFrontendPreview, True Float=600, False Float=0)
    Cylinder Height Midpoint=0.888
    Cylinder Radius=0
    MaskSpawnGroup=false
    Override Seed=false
    Rotation Mode=::NewEnumerator1
    Shape Primitive=::NewEnumerator1
    Surface Only=false
    UseOwnerScale=true
   )
   &DLR_ProductPreviews(GameModeID:NiagaraInt32=$System.GameModeID, bPreviewVehicleSpeed:NiagaraBool=$System.bPreviewVehicleSpeed)
  ]
  particleUpdateScript [
   &ParticleState()
   &SpriteRotationRate(Rotation Rate=&RandomRangeFloat(Minimum=-80, Maximum=80)) {module=SpriteRotationRate001}
   &ScaleVelocity(
    Coordinate Space:ENiagaraCoordinateSpace=::Simulation
    Velocity Scale=&MakeCustomVectorFromBool(
     Bool=$User.bIsFrontendPreview
     True Vector=&VectorFromCurve(VectorCurve=DataInterface() {propertyValues {XCurve {keys [<at -0.015 1.5338>,<at 0.25 0.4641>,<at 0.8 0.177>]}, YCurve {keys [<at 0 1>,<at 1 1>]}, ZCurve {keys [<at 0 1>,<at 1 1>]}}})
     False Vector=<1 1 1>
    )
   )
   &ScaleSpriteSize(
    Scale Sprite Size Mode=::NewEnumerator0
    Uniform Scale Factor=&MakeCustomFloatFromBool(
     Bool=$User.bIsFrontendPreview
     True Float=&FloatFromCurve(FloatCurve=DataInterface() {propertyValues {Curve {keys [
          {arriveTangent=1.3713, interpMode=RCIM_Cubic, leaveTangent=1.3713, tangentMode=RCTM_User, time=0, value=1.2}
          {arriveTangent=1.5367, interpMode=RCIM_Cubic, leaveTangent=1.5367, time=0.5079, value=1.7}
          {arriveTangent=2.0823, tangentMode=RCTM_User, time=1.01, value=2.752}
         ]}}})
     False Float=&FloatFromCurve(FloatCurve=DataInterface() {propertyValues {Curve {keys [
          {arriveTangent=-0.2005, interpMode=RCIM_Cubic, leaveTangent=-0.2005, tangentMode=RCTM_User, time=0.5, value=1.7}
          {arriveTangent=-4.8004, tangentMode=RCTM_User, time=1, value=0.5}
         ]}}})
    )
   )
   &SpriteRotationRate(Rotation Rate=&Multiply_Float(A=&RandomRangeFloat(Minimum=0.25, Maximum=1), B=25))
   &DynamicMaterialParameters(
    Index 0 Param 1=1
    Index 1 Param 1=&SelectFloatFromArray(Direct Array Index=$System.ColorVariantID, Float Selection Array=DataInterface() {propertyValues {FloatData [1.3754,1.3754,1.3754,2.5,6]}})
    Index 1 Param 2=&FloatFromCurve(FloatCurve=DataInterface() {propertyValues {Curve {keys [<at -0.0054 -0.2952>,<at 0.1771 0>,<at 0.5125 0.6394>,<at 0.9958 2.4066>]}}})
    Index 2 Param 1=&MakeCustomFloatFromBool(
     Bool=$User.bIsFrontendPreview
     True Float=&FloatFromCurve(FloatCurve=DataInterface() {propertyValues {Curve {keys [<at 0.3104 0.2007>,<at 0.6882 0.2012>,<at 1.0118 0.2055>]}}})
     False Float=&FloatFromCurve(FloatCurve=DataInterface() {propertyValues {Curve {keys [<at 0.3037 0.2964>,<at 0.7057 0.2974>,<at 0.9989 0.3141>]}}})
    )
    Index 2 Param 2=&MakeCustomFloatFromBool(
     Bool=$User.bIsFrontendPreview
     True Float=&Add_Float(
      A=&FloatFromCurve(FloatCurve=DataInterface() {propertyValues {Curve {keys [<at -0.0012 0.5>,<at 0.4935 0.1862>,<at 0.9967 -0.1566>]}}})
      B=&RandomRangeFloat(Minimum=0, Maximum=1)
     )
     False Float=&Add_Float(
      A=&FloatFromCurve(FloatCurve=DataInterface() {propertyValues {Curve {keys [<at -0.0066 0.8>,<at 0.5043 0.4>,<at 0.9967 -0.1566>]}}})
      B=&RandomRangeFloat(Minimum=0, Maximum=1)
     )
    )
    Index 2 Param 3=&RandomRangeFloat(Minimum=0, Maximum=1)
    Index 2 Param 4=&RandomRangeFloat(Minimum=0, Maximum=1)
    Index 3 Param 1=0.2231
    Module.Index0Float0Write=false
    Module.Index0Float1Write=true
    Module.Index0Float2Write=true
    Module.Index0Float3Write=false
    Module.Index1Float0Write=true
    Module.Index1Float1Write=true
    Module.Index1Float2Write=true
    Module.Index1Float3Write=true
    Module.Index2Float0Write=true
    Module.Index2Float1Write=true
    Module.Index2Float2Write=true
    Module.Index2Float3Write=true
    Module.Index3Float0Write=true
    Module.Index3Float1Write=true
    Module.Index3Float2Write=true
    Module.Index3Float3Write=true
    Write Parameter Index 1=::NewEnumerator1
    Write Parameter Index 2=::NewEnumerator1
    Write Parameter Index 3=::NewEnumerator1
   )
   &InheritSourceMovement(
    Applied Position Scale=&Multiply_VectorByFloat(
     Vector=<1 1 1>
     Float=&FloatFromCurve(FloatCurve=DataInterface() {propertyValues {Curve {
         defaultValue=1
         keys [
          {arriveTangent=-1.7885, interpMode=RCIM_Cubic, leaveTangent=-1.7885, tangentMode=RCTM_User, time=0.05, value=1}
          {interpMode=RCIM_Cubic, tangentMode=RCTM_User, time=1, value=0}
         ]
        }}})
    )
    ApplyForce=false
    ApplyPosition=true
    ApplyPositionFalloff=false
   ) {module=InheritSourceMovement001}
   &ScaleColor(
    Scale Alpha=&MakeCustomFloatFromBool(
     Bool=$User.bIsFrontendPreview
     True Float=&FloatFromCurve(FloatCurve=DataInterface() {propertyValues {Curve {keys [<at -0.0051 0.0115>,<at 0.2082 0.9>,<at 0.8115 0.913>,<at 1.0183 0.9>]}}})
     False Float=&FloatFromCurve(FloatCurve=DataInterface() {propertyValues {Curve {keys [<at 0 1>,<at 0.5 0.5>,<at 0.9923 0>]}}})
    )
    ScaleA=true
    ScaleRGB=true
   )
   &SolveForcesAndVelocity(Clamp Velocity=false, Limit Acceleration=false)
  ]
  rendererClasses [&NiagaraSpriteRendererProperties]
  renderers [
   {rendererClass=&NiagaraSpriteRendererProperties}
   {rendererClass=&NiagaraSpriteRendererProperties, rendererIndex=1}
  ]
 }
 NE_VC_softSmokePuffs {
  emitterUpdateScript [
   &EmitterState()
   &VC_FilteredSocketCount()
   &SpawnRate(
    SpawnRate=&MakeCustomFloatFromBool(Bool=$User.bIsFrontendPreview, True Float=75, False Float=100)
    Use Spawn Probability=false
   )
  ]
  particleSpawnScript [
   &VC_PaintedVariantsOverride(BurntSienna=<rgba 0.25 0.0374 0 1>, Cobalt=<rgba 0 0.0071 0.3854 1>)
   &InitializeParticle(
    Color=&MakeCustomLinearColorFromBool(Bool=$User.bIsPainted, True Color=$Particles.Color, False Color=<rgba 0.0156 0.0156 0.0156 1>)
    Lifetime=&MakeCustomFloatFromBool(
     Bool=$User.bIsFrontendPreview
     True Float=&RandomRangeFloat(Minimum=0.8, Maximum=1)
     False Float=&RandomRangeFloat(Minimum=0.5, Maximum=0.7)
    )
    Position Offset=<-300 0 0>
    Sprite Rotation Mode=::NewEnumerator2
    Sprite Size Mode=::NewEnumerator3
    Uniform Sprite Size=&RandomRangeFloat(Minimum=70, Maximum=90)
    UsePositionOffset=false
    Write Material Random=true
   )
   &PartitionParticles(Number Of Partitions=$Emitter.PortCount)
   &SkeletalMeshLocation(
    Bone / Socket Index=$Particles.PartitionParticles.Partition
    Bone / Socket Sampling Mode=::NewEnumerator5
    MaskSpawnGroup=false
    Mesh Sampling Type=::NewEnumerator6
    Skeletal Mesh=$System.SkeletalMesh
   )
   &AddVelocity(Velocity=&MakeCustomVectorFromBool(
     Bool=$User.bIsFrontendPreview
     True Vector=&RandomRangeVector(Minimum=<-200 0 0>, Maximum=<-700 0 0>)
     False Vector=&RandomRangeVector(Minimum=<-800 0 0>, Maximum=<-1400 0 0>)
    ))
   &ShapeLocation(
    Cylinder Height=&MakeCustomFloatFromBool(Bool=$User.bIsFrontendPreview, True Float=360, False Float=0)
    Cylinder Height Midpoint=0.9493
    Cylinder Radius=0
    MaskSpawnGroup=false
    Override Seed=false
    Rotation Mode=::NewEnumerator1
    Shape Primitive=::NewEnumerator1
    Surface Only=false
    UseOwnerScale=true
   )
   &DLR_ProductPreviews(GameModeID:NiagaraInt32=$System.GameModeID, bPreviewVehicleSpeed:NiagaraBool=$System.bPreviewVehicleSpeed)
  ]
  particleUpdateScript [
   &ParticleState()
   &SpriteRotationRate(Rotation Rate=&RandomRangeFloat(Minimum=-80, Maximum=80)) {module=SpriteRotationRate001}
   &ScaleVelocity(
    Coordinate Space:ENiagaraCoordinateSpace=::Simulation
    Velocity Scale=&MakeCustomVectorFromBool(
     Bool=$User.bIsFrontendPreview
     True Vector=&VectorFromCurve(VectorCurve=DataInterface() {propertyValues {XCurve {keys [<at -0.0116 2>,<at 0.366 0.5>,<at 0.7 0.195>]}, YCurve {keys [<at 0 1>,<at 1 1>]}, ZCurve {keys [<at 0 1>,<at 1 1>]}}})
     False Vector=<1 1 1>
    )
   )
   &ScaleSpriteSize(Uniform Curve Sprite Scale=DataInterface() {propertyValues {Curve {keys [
        {arriveTangent=-0.0803, interpMode=RCIM_Cubic, time=0.1238, value=2.4}
        {arriveTangent=-0.3866, tangentMode=RCTM_User, time=0.9939, value=1.881}
       ]}}})
   &SpriteRotationRate(Rotation Rate=&Multiply_Float(A=&RandomRangeFloat(Minimum=0.25, Maximum=1), B=20))
   &SubUVAnimation(
    Random Start Frame=true
    "Sprite Renderer "=DataInterface() {propertyValues {SpriteRenderer=&self:NE_VC_softSmokePuffs_0.NiagaraSpriteRendererProperties_0}}
    SubUV Animation Mode:ENiagara_SubUVLookupModeV2=::NewEnumerator1
    SubUV Curve=DataInterface() {propertyValues {Curve {keys [
        {arriveTangent=1.0632, interpMode=RCIM_Cubic, leaveTangent=1.0632, tangentMode=RCTM_User, time=0, value=0}
        {arriveTangent=0.6316, interpMode=RCIM_Cubic, leaveTangent=0.6316, tangentMode=RCTM_User, time=0.4972, value=0.3}
        {arriveTangent=0.4083, interpMode=RCIM_Cubic, leaveTangent=0.4083, tangentMode=RCTM_User, time=1.0028, value=0.4834}
       ]}}}
    UseEndFrame=false
    UseStartFrame=false
   ) {module=SubUVAnimation001}
   &InheritSourceMovement(
    Applied Position Scale=&Multiply_VectorByFloat(
     Vector=<1 1 1>
     Float=&FloatFromCurve(FloatCurve=DataInterface() {propertyValues {Curve {
         defaultValue=1
         keys [
          {arriveTangent=-1.7885, interpMode=RCIM_Cubic, leaveTangent=-1.7885, tangentMode=RCTM_User, time=0.05, value=1}
          {interpMode=RCIM_Cubic, tangentMode=RCTM_User, time=1, value=0}
         ]
        }}})
    )
    ApplyForce=false
    ApplyPosition=true
    ApplyPositionFalloff=false
   ) {module=InheritSourceMovement001}
   &ScaleColor(
    Scale Alpha=&MakeCustomFloatFromBool(
     Bool=$User.bIsFrontendPreview
     True Float=&FloatFromCurve(FloatCurve=DataInterface() {propertyValues {Curve {keys [<at 0.0007 0>,<at 0.1651 0.5712>,<at 0.6774 0.3924>,<at 0.9959 0>]}}})
     False Float=&FloatFromCurve(FloatCurve=DataInterface() {propertyValues {Curve {keys [<at -0.021 0.988>,<at 0.5 0.5>,<at 0.9959 0>]}}})
    )
    ScaleA=true
    ScaleRGB=true
   )
   &SolveForcesAndVelocity(Clamp Velocity=false, Limit Acceleration=false)
  ]
  rendererClasses [&NiagaraSpriteRendererProperties]
  renderers [{rendererClass=&NiagaraSpriteRendererProperties}]
 }
}
summary {
 systemName=NS_BoostInk
 emitters [
  {emitterName=NE_VC_Boost_Ribbon_Parent, rendererClasses [&NiagaraRibbonRendererProperties]}
  {emitterName=NE_VC_softSmokePuffs, rendererClasses [&NiagaraSpriteRendererProperties]}
  {emitterName=NE_VC_InkPuffs, rendererClasses [&NiagaraSpriteRendererProperties]}
 ]
 userVariables [
  User.bIsFrontendPreview {type=NiagaraBool}
  User.bIsPainted {type=NiagaraBool}
  User.blsIsCoroColor {type=NiagaraBool}
  User.bPreviewVehicleSpeed {type=NiagaraBool}
  User.ColorVariantID {type=NiagaraInt32}
  User.DTCarOffset {type=Vector3f, description="", defaultValue {struct=Vector3f, value=<0 0 0>}}
  User.GameModeID {type=NiagaraInt32}
  User.PaintedColor {type=LinearColor, description="", defaultValue {struct=LinearColor, value=<rgba 1 1 1 1>}}
  User.PaintedScalar {type=NiagaraFloat, defaultValue=1, description=""}
  User.PortCount {type=NiagaraInt32, defaultValue=2}
  User.Scale {type=NiagaraFloat, defaultValue=2, description=""}
  User.SingleBoostPos {type=Vector3f, description="", defaultValue {struct=Vector3f, value=<0 0 0>}}
 ]
}
